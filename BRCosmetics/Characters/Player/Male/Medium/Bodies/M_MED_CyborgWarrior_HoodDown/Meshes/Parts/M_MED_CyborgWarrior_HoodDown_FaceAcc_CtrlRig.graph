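# Corrective Control Rig for the male medium CyborgWarrior's lowered hood, offsetting hood bones as head, arms and stance change.
# It first derives CrouchingButUp from pelvis height through a remap and curve.
# Spherical pose readers on the head and upper arms then weight side hood offsets, with head ones gated by Targeting and Crouching.
# A swing-twist DeltaRotation function converts head rotation against head_null into weights for further hood offsets, some similarly gated.
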ControlRigBlueprint /BRCosmetics/Characters/Player/Male/Medium/Bodies/M_MED_CyborgWarrior_HoodDown/Meshes/Parts/M_MED_CyborgWarrior_HoodDown_FaceAcc_CtrlRig
graphs [
 Multiply_1_ContainedGraph {nodes [
   n1 = RigVMFunctionEntryNode()
   Multiply = &RigVMFunction_MathDoubleMul(A=@n1.A, B=@n1.B)
   RigVMFunctionReturnNode(Result=&RigVMFunction_MathDoubleMul(B=@n1.C, A=@Multiply.Result).Result)
  ]}
 Multiply_1_ContainedGraph {nodes [
   n1 = RigVMFunctionEntryNode()
   Multiply = &RigVMFunction_MathDoubleMul(A=@n1.A, B=@n1.B)
   RigVMFunctionReturnNode(Result=&RigVMFunction_MathDoubleMul(B=@n1.D, A=&RigVMFunction_MathDoubleMul(B=@n1.C, A=@Multiply.Result).Result).Result)
  ]}
 Multiply_3_ContainedGraph {nodes [
   n1 = RigVMFunctionEntryNode()
   n2 = &RigVMFunction_MathDoubleMul(A=@n1.A, B=@n1.B)
   RigVMFunctionReturnNode(Result=&RigVMFunction_MathDoubleMul(B=@n1.D, A=&RigVMFunction_MathDoubleMul(B=@n1.C, A=@n2.Result).Result).Result)
  ]}
 New Function_ContainedGraph {nodes [
   n1 = RigVMFunctionEntryNode()
   RigVMFunction_MathBoolToInteger = &RigVMFunction_MathBoolToInteger(Value=@n1.onlyTwist)
   RigVMFunction_MathQuaternionSwingTwist = &RigVMFunction_MathQuaternionSwingTwist(
    twistAxis=@n1.twistAxis
    Input=&RigUnit_GetRelativeTransformForItem(
     bChildInitial=False
     bParentInitial=False
     child=(Type=Bone,Name="None")
     parent=(Type=Bone,Name="None")
     child.Name=@n1.child
     parent.Name=@n1.parent
     parent.Type=@n1.parentType
    ).RelativeTransform.Rotation
   )
   RigVMFunction_MathQuaternionToEuler = &RigVMFunction_MathQuaternionToEuler(
    Value=RigVMDispatchNode(
     Values=((),())
     Index=@RigVMFunction_MathBoolToInteger.Result
     Values.0=@RigVMFunction_MathQuaternionSwingTwist.Swing
     Values.1=@RigVMFunction_MathQuaternionSwingTwist.Twist
    ) {template="DISPATCH_RigVMDispatch_SelectInt32(in Index,in Values,out Result)"}.Result
    RotationOrder=@n1.Rotation Order
   )
   > DISPATCH_RigVMDispatch_Print = RigVMDispatchNode(ScreenColor=(R=1,G=1,B=1,A=1), ScreenDuration=1.5, Enabled=@n1.debugRotation, Value=@RigVMFunction_MathQuaternionToEuler.Result) {template="DISPATCH_RigVMDispatch_Print(in Prefix,in Value,in Enabled,in ScreenDuration,in ScreenColor)"}
   Clamp = &RigVMFunction_MathIntClamp(Value=@n1.primAxis)
   DISPATCH_RigVMDispatch_SelectInt32 = RigVMDispatchNode(
    Values=(0,0,0)
    Index=@Clamp.Result
    Values.0=@RigVMFunction_MathQuaternionToEuler.Result.X
    Values.1=@RigVMFunction_MathQuaternionToEuler.Result.Y
    Values.2=@RigVMFunction_MathQuaternionToEuler.Result.Z
   ) {template="DISPATCH_RigVMDispatch_SelectInt32(in Index,in Values,out Result)"}
   Remap = &RigVMFunction_MathDoubleRemap(
    TargetMaximum=1
    TargetMinimum=0
    bClamp=False
    Value=@DISPATCH_RigVMDispatch_SelectInt32.Result
    SourceMaximum=@n1.max
    SourceMinimum=@n1.min
   )
   Multiply = &RigVMFunction_MathDoubleMul(A=&RigVMFunction_MathDoubleClamp(Value=@Remap.Result).Result, B=@n1.premult)
   > RigVMFunctionReturnNode(outputWeight=@Multiply.Result)
  ]}
 RigVMFunctionLibrary {nodes [DeltaRotation = RigVMCollapseNode(ExecuteContext=(), Rotation Order=ZYX, premult=1, twistAxis=(X=1,Y=0,Z=0))]}
 RigVMFunction_Sequence_1_ContainedGraph {nodes [
   n1 = RigVMFunctionEntryNode()
   RigVMFunction_Sequence = &RigVMFunction_Sequence(ExecuteContext=@n1.ExecuteContext)
   n2 = &RigVMFunction_Sequence(ExecuteContext=@RigVMFunction_Sequence.B)
   RigVMFunctionReturnNode(A=@RigVMFunction_Sequence.A, B=@n2.A, C=@n2.B)
  ]}
 RigVMModel {nodes [
   RigUnit_BeginExecution = &RigUnit_BeginExecution()
   RigVMFunction_Sequence = RigVMAggregateNode(ExecuteContext=@RigUnit_BeginExecution.ExecutePin)
   RigVMFunction_MathBoolMake = /Script/RigVM.RigVMFunction_MathBoolMake(Value=False) {is=struct RigVMUnitNode}
   n1 = &RigUnit_SphericalPoseReader(
    ActiveRegionScaleFactors=(PositiveWidth=0.5,NegativeWidth=0.3,PositiveHeight=0.5,NegativeHeight=0.5)
    ActiveRegionSize=0.1
    Debug=(bDrawDebug=true,bDraw2D=False,bDrawLocalAxes=False,DebugScale=25,DebugSegments=20,DebugThickness=0.25)
    DriverAxis=(X=0,Y=1,Z=0)
    DriverItem=(Type=Bone,Name="head")
    FalloffRegionScaleFactors=(PositiveWidth=0.3,NegativeWidth=0.1,PositiveHeight=0.3,NegativeHeight=0.3)
    FalloffSize=0.2
    FlipHeightScaling=false
    FlipWidthScaling=false
    OptionalParentItem=(Type=Bone,Name="None")
    RotationOffset=(X=100,Y=0,Z=-8)
    Debug.bDrawDebug=@RigVMFunction_MathBoolMake.Value
    ExecutePin=@RigVMFunction_Sequence.B
   )
   RigVMFunction_MathBoolNot = &RigVMFunction_MathBoolNot(Value=RigVMVariableNode() {variable=Crouching}.Value)
   RigVMFunction_MathBoolToFloat = &RigVMFunction_MathBoolToFloat(Value=@RigVMFunction_MathBoolNot.Result)
   Multiply = RigVMAggregateNode(
    C=@RigVMFunction_MathBoolToFloat.Result
    B=&RigVMFunction_MathBoolToFloat(Value=RigVMVariableNode() {variable=Targeting}.Value).Result
    A=@n1.OutputParam
   )
   > &RigUnit_ModifyTransforms(
    ItemToModify=((Item=(Type=Bone,Name="base_hood_off_l"),Transform=(Rotation=(X=0,Y=0.1736,Z=0,W=0.9848),Translation=(X=0,Y=0,Z=0),Scale3D=(X=1,Y=1,Z=1))))
    WeightMinimum=0
    Weight=@Multiply.Result
   )
   > n2 = &RigUnit_SphericalPoseReader(
    ActiveRegionScaleFactors=(PositiveWidth=1,NegativeWidth=0.5,PositiveHeight=0.75,NegativeHeight=0.75)
    ActiveRegionSize=0.1
    Debug=(bDrawDebug=true,bDraw2D=False,bDrawLocalAxes=False,DebugScale=25,DebugSegments=20,DebugThickness=0.25)
    DriverAxis=(X=0,Y=1,Z=0)
    DriverItem=(Type=Bone,Name="head")
    FalloffRegionScaleFactors=(PositiveWidth=1,NegativeWidth=0.1,PositiveHeight=1,NegativeHeight=1)
    FalloffSize=0.2
    FlipHeightScaling=False
    FlipWidthScaling=False
    OptionalParentItem=(Type=Bone,Name="None")
    RotationOffset=(X=100,Y=0,Z=-30)
    Debug.bDrawDebug=@RigVMFunction_MathBoolMake.Value
   )
   Multiply_1 = RigVMAggregateNode(
    C=&RigVMFunction_MathBoolToFloat(Value=RigVMVariableNode() {variable=Crouching}.Value).Result
    B=&RigVMFunction_MathBoolToFloat(Value=RigVMVariableNode() {variable=Targeting}.Value).Result
    D=RigVMVariableNode() {variable=CrouchingButUp}.Value
    A=@n2.OutputParam
   )
   > &RigUnit_ModifyTransforms(
    ItemToModify=((Item=(Type=Bone,Name="base_hood_off_l"),Transform=(Rotation=(X=0,Y=-0.0872,Z=0,W=0.9962),Translation=(X=-5,Y=0,Z=0),Scale3D=(X=1,Y=1,Z=1))))
    WeightMinimum=0
    Weight=@Multiply_1.Result
   )
   > n3 = &RigUnit_SphericalPoseReader(
    ActiveRegionScaleFactors=(PositiveWidth=2,NegativeWidth=2,PositiveHeight=1,NegativeHeight=1)
    ActiveRegionSize=0.05
    Debug=(bDrawDebug=True,bDraw2D=False,bDrawLocalAxes=False,DebugScale=25,DebugSegments=20,DebugThickness=0.25)
    DriverAxis=(X=1,Y=0,Z=0)
    DriverItem=(Type=Bone,Name="upperarm_l")
    FalloffRegionScaleFactors=(PositiveWidth=1,NegativeWidth=1,PositiveHeight=0.75,NegativeHeight=0.75)
    FalloffSize=0.2
    FlipHeightScaling=False
    FlipWidthScaling=False
    OptionalParentItem=(Type=Bone,Name="None")
    RotationOffset=(X=10,Y=-50,Z=0)
    Debug.bDrawDebug=@RigVMFunction_MathBoolMake.Value
   )
   > &RigUnit_ModifyTransforms(
    ItemToModify=((Item=(Type=Bone,Name="base_hood_off_l"),Transform=(Rotation=(X=0,Y=0,Z=0,W=1),Translation=(X=0.8,Y=0,Z=0),Scale3D=(X=1,Y=1,Z=1))))
    WeightMinimum=0
    Weight=@n3.OutputParam
   )
   > n4 = &RigUnit_SphericalPoseReader(
    ActiveRegionScaleFactors=(PositiveWidth=2,NegativeWidth=2,PositiveHeight=1,NegativeHeight=1)
    ActiveRegionSize=0.05
    Debug=(bDrawDebug=True,bDraw2D=False,bDrawLocalAxes=False,DebugScale=25,DebugSegments=20,DebugThickness=0.25)
    DriverAxis=(X=-1,Y=0,Z=0)
    DriverItem=(Type=Bone,Name="upperarm_r")
    FalloffRegionScaleFactors=(PositiveWidth=1,NegativeWidth=1,PositiveHeight=0.75,NegativeHeight=0.75)
    FalloffSize=0.2
    FlipHeightScaling=False
    FlipWidthScaling=False
    OptionalParentItem=(Type=Bone,Name="None")
    RotationOffset=(X=-10,Y=130,Z=0)
    Debug.bDrawDebug=@RigVMFunction_MathBoolMake.Value
   )
   > &RigUnit_ModifyTransforms(
    ItemToModify=((Item=(Type=Bone,Name="base_hood_off_r"),Transform=(Rotation=(X=0,Y=0,Z=0,W=1),Translation=(X=0.8,Y=0,Z=0),Scale3D=(X=1,Y=1,Z=1))))
    WeightMinimum=0
    Weight=@n4.OutputParam
   )
   DeltaRotation_6 = RigVMFunctionReferenceNode(
    Rotation Order=ZYX
    child=head
    debugRotation=false
    max=-19
    min=-5
    parent=head_null
    parentType=Null
    premult=1
    primAxis=2
    twistAxis=(X=1,Y=0,Z=0)
    ExecuteContext=@RigVMFunction_Sequence.C
   )
   > &RigUnit_ModifyTransforms(
    ItemToModify="((Item=(Type=Bone,Name=\"dyn_hood_off_r\"),Transform=(Rotation=(X=0.0151,Y=0.0858,Z=-0.173,W=0.9811),Translation=(X=-2,Y=0,Z=0),Scale3D=(X=1,Y=1,Z=1))),(Item=(Type=Bone,Name=\"dyn_hood_off_l\"),Transform=(Rotation=(X=0,Y=0,Z=0,W=1),Translation=(X=0,Y=0,Z=0),Sc ...(truncated)"
    WeightMinimum=0
    Weight=@DeltaRotation_6.outputWeight
   )
   Multiply_2 = RigVMAggregateNode(
    A=&RigVMFunction_MathBoolToFloat(Value=RigVMVariableNode() {variable=Targeting}.Value).Result
    B=&RigVMFunction_MathBoolToFloat(Value=RigVMVariableNode() {variable=Crouching}.Value).Result
    C=RigVMVariableNode() {variable=CrouchingButUp}.Value
    D=@DeltaRotation_6.outputWeight
   )
   > &RigUnit_ModifyTransforms(
    ItemToModify="((Item=(Type=Bone,Name=\"base_hood_off_01b\"),Transform=(Rotation=(X=0,Y=0.342,Z=0,W=0.9397),Translation=(X=0,Y=0,Z=-3),Scale3D=(X=1,Y=1,Z=1))),(Item=(Type=Bone,Name=\"base_hood_off_01a\"),Transform=(Rotation=(X=0,Y=0,Z=0,W=1),Translation=(X=0,Y=0,Z=0),Scale3D ...(truncated)"
    WeightMinimum=0
    Weight=@Multiply_2.Result
   )
   RigUnit_GetRelativeTransformForItem = &RigUnit_GetRelativeTransformForItem(Child=(Type=Bone,Name="pelvis"), Parent=(Type=Bone,Name="root"), bChildInitial=False, bParentInitial=False)
   Remap = &RigVMFunction_MathDoubleRemap(
    SourceMaximum=30
    SourceMinimum=60
    TargetMaximum=0
    TargetMinimum=1
    bClamp=False
    Value=@RigUnit_GetRelativeTransformForItem.RelativeTransform.Translation.Z
   )
   n5 = &RigVMFunction_AnimEvalRichCurve(
    Curve="(EditorCurveData=(Keys=((InterpMode=RCIM_Cubic,TangentMode=RCTM_Break,TangentWeightMode=RCTWM_WeightedNone,Time=0,Value=0,ArriveTangent=0,ArriveTangentWeight=0,LeaveTangent=0,LeaveTangentWeight=0),(InterpMode=RCIM_Cubic,TangentMode=RCTM_Break,TangentWeight ...(truncated)"
    SourceMinimum=0
    TargetMinimum=0
    Value=@Remap.Result
   )
   RigVMVariableNode(Value=@n5.Result, ExecuteContext=@RigVMFunction_Sequence.A) {variable=CrouchingButUp}
  ]}
]
hierarchy [
 root {loc=<0 0 0>, rot=<0 0 0 1>}
 pelvis {parent=root, loc=<0 -0.7071 93.1191>, rot=<0.0127 0.707 -0.0127 -0.707>}
 spine_01 {parent=pelvis, loc=<6.7494 -0.6388 0>, rot=<0 0 0.034 -0.9994>}
 spine_02 {parent=spine_01, loc=<7.2518 -0.035 0>, rot=<0 0 0.0238 -0.9997>}
 spine_03 {parent=spine_02, loc=<6.0104 0 0>, rot=<0 0 0.0149 -0.9999>}
 spine_04 {parent=spine_03, loc=<8.1024 0 0>, rot=<0 0 -0.0132 -0.9999>}
 spine_05 {parent=spine_04, loc=<10.0146 0 0>, rot=<0 0 -0.0015 -1>}
 neck_01 {parent=spine_05, loc=<11.3138 0.1529 0>, rot=<0 0 -0.2003 -0.9797>}
 neck_02 {parent=neck_01, loc=<4.5627 0 0>, rot=<0 0 0.0504 -0.9987>}
 head {parent=neck_02, loc=<5.3914 0 0>, rot=<0 0 0.115 -0.9934>}
 base_hood_off_01b {parent=spine_05, loc=<17.8541 -6.2814 -0.4625>, rot=<0.5555 -0.4376 0.4376 -0.5555>}
 dyn_hood_off_01b {parent=base_hood_off_01b, loc=<0 0 0>, rot=<0 0 0 1>}
 base_hood_off_01a {parent=spine_05, loc=<15.1469 -7.1273 -0.3979>, rot=<0.2713 -0.6606 0.6428 -0.2772>}
 dyn_hood_off_01a {parent=base_hood_off_01a, loc=<0 0 0>, rot=<0 0 0 1>}
 dyn_hood_off_02 {parent=dyn_hood_off_01a, loc=<4.674 0 0>, rot=<0 -0.151 0 -0.9885>}
 base_hood_off_l {parent=spine_05, loc=<15.2032 1.2465 -10.7662>, rot=<-0.6515 0.0802 -0.3717 -0.6565>}
 dyn_hood_off_l {parent=base_hood_off_l, loc=<0 0 0>, rot=<0 0 0 1>}
 base_hood_off_r {parent=spine_05, loc=<15.2737 1.0696 10.3562>, rot=<0.6177 -0.0601 -0.3408 -0.7061>}
 dyn_hood_off_r {parent=base_hood_off_r, loc=<0 0 0>, rot=<0 0 0 1>}
 clavicle_l {parent=spine_05, loc=<4.3575 -0.1466 -2.3223>, rot=<0.0465 -0.7289 -0.009 -0.6829>}
 upperarm_l {parent=clavicle_l, loc=<15.5111 0 0>, rot=<0.0226 -0.3234 -0.0088 -0.9459>}
 upperarm_twist_01_l {parent=upperarm_l, loc=<0.0041 0 0>, rot=<0 0 0 -1>}
 upperarm_twist_02_l {parent=upperarm_l, loc=<13.7941 0 0>, rot=<0 0 0 -1>}
 clavicle_r {parent=spine_05, loc=<4.3575 -0.1466 2.3223>, rot=<0.7289 0.0465 -0.6829 0.009>}
 upperarm_r {parent=clavicle_r, loc=<-15.5111 0 0>, rot=<0.0226 -0.3234 -0.0088 -0.9459>}
 upperarm_twist_01_r {parent=upperarm_r, loc=<-0.0041 0 0>, rot=<0 0 0 -1>}
 upperarm_twist_02_r {parent=upperarm_r, loc=<-13.7941 0 0>, rot=<0 0 0 -1>}
 head_null {parent=spine_05, type=Null, loc=<20.6573 3.5489 0>, rot=<0 0 -0.036 -0.9994>}
]
variables [
 Crouching {type=bool, default=False}
 CrouchingButUp {type=double, default=0}
 Targeting {type=bool, default=False}
]
