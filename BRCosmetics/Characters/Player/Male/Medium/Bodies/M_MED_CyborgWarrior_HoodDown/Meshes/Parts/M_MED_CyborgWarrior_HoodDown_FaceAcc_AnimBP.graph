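# Animation Blueprint for the male Cyborg Warrior hood-down face accessory, adding secondary motion to the lowered hood.
# It copies the attached parent's pose, layering in clavicle and upper-arm branches from the body mesh cached on customization load.
# Three sprung AnimDynamics nodes simulate hood bones, active only while IsWearingBackpackOrCape is false.
# A Control Rig driven by crouching and targeting states produces the final pose.
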AnimBlueprint /BRCosmetics/Characters/Player/Male/Medium/Bodies/M_MED_CyborgWarrior_HoodDown/Meshes/Parts/M_MED_CyborgWarrior_HoodDown_FaceAcc_AnimBP
target_skeleton=&/M_MED_CyborgWarrior_HoodDown_FaceAcc_Skeleton
graphs [
 AnimGraph {nodes [
   n1 = NOT Boolean(A=Get IsWearingBackpackOrCape() -> (IsWearingBackpackOrCape:bool))
   n2 = Anim Dynamics - Bone: dyn_hood_off_01b!(
    ComponentPose=Local To Component(LocalPose=Use cached pose 'StartPose'())
    bAlphaBoolEnabled:bool=@n1
    GravityOverride:Vector="0,0,-980"
    ExternalForce="0,0,0"
   ) -> (Pose) {
    alphaInputType=Bool
    angularDampingOverride=1
    angularSpringConstant=11000
    bAngularSpring=true
    bOverrideAngularDamping=true
    bUseGravityOverride=true
    gravityScale=0.35
    simulationSpace=RootRelative
    alphaBoolBlend {blendOption=Linear}
    boundBone {boneName=dyn_hood_off_01b}
    componentAppliedLinearAccClamp=<9500 9500 9500>
    componentLinearAccScale=<0.5 0.5 0.5>
    componentLinearVelScale=<0.5 0.5 0.5>
    gravityOverride {z=-980}
    physicsBodyDefinitions [{
      boundBone {boneName=dyn_hood_off_01b}
      boxExtents=<20 30 30>
      constraintSetup {angularLimitsMax=<1 1 3>, angularLimitsMin=<-1 -90 -3>, angularTarget=<1 0 0>}
      localJointOffset=<6 0 0>
     }]
   }
   n3 = Anim Dynamics - Bone: dyn_hood_off_01a!(ComponentPose=@n2, bAlphaBoolEnabled:bool=@n1, GravityOverride:Vector="0,0,-980", ExternalForce="0,0,0") -> (Pose) {
    alphaInputType=Bool
    angularDampingOverride=1
    angularSpringConstant=11000
    bAngularSpring=true
    bOverrideAngularDamping=true
    bUseGravityOverride=true
    gravityScale=0.35
    simulationSpace=RootRelative
    alphaBoolBlend {blendOption=Linear}
    boundBone {boneName=dyn_hood_off_01a}
    componentAppliedLinearAccClamp=<9500 9500 9500>
    componentLinearAccScale=<0.5 0.5 0.5>
    componentLinearVelScale=<0.5 0.5 0.5>
    gravityOverride {z=-980}
    physicsBodyDefinitions [{
      boundBone {boneName=dyn_hood_off_01a}
      boxExtents=<20 30 30>
      constraintSetup {angularLimitsMax=<1 1 3>, angularLimitsMin=<-1 -90 -3>, angularTarget=<1 0 0>}
      localJointOffset=<6 0 0>
     }]
   }
   n4 = Anim Dynamics - Bone: dyn_hood_off_02!(ComponentPose=@n3, bAlphaBoolEnabled:bool=@n1, GravityOverride:Vector="0,0,-980", ExternalForce="0,0,0") -> (Pose) {
    alphaInputType=Bool
    angularDampingOverride=1
    angularSpringConstant=11000
    bAngularSpring=true
    bOverrideAngularDamping=true
    bUseGravityOverride=true
    gravityScale=0.35
    simulationSpace=RootRelative
    alphaBoolBlend {blendOption=Linear}
    boundBone {boneName=dyn_hood_off_02}
    componentAppliedLinearAccClamp=<9500 9500 9500>
    componentLinearAccScale=<0.5 0.5 0.5>
    componentLinearVelScale=<0.5 0.5 0.5>
    gravityOverride {z=-980}
    physicsBodyDefinitions [{
      boundBone {boneName=dyn_hood_off_02}
      boxExtents=<20 30 30>
      constraintSetup {angularLimitsMax=<1 1 1>, angularLimitsMin=<-5 -60 -1>, angularTarget=<1 0 0>}
      localJointOffset=<4 0 0>
     }]
   }
   n5 = Control Rig!(Source=Component To Local(ComponentPose=@n4), Alpha, Crouching:bool=Get bIsCrouching(), Targeting:bool=Get bIsTargeting()) -> (Pose) {
    alphaBoolBlend {blendOption=Linear}
    controlRigAssetReference {blueprintRigClass=&/M_MED_CyborgWarrior_HoodDown_FaceAcc_CtrlRig.M_MED_CyborgWarrior_HoodDown_FaceAcc_CtrlRig_C}
   }
   Output Pose(Result=@n5) {layerGroup=None}
   n6 = Layered blend per bone(
    BasePose=Copy Pose From Mesh() {bCopyCurves=true, bUseAttachedParent=true}
    BlendPoses_0=Copy Pose From Mesh(SourceMeshComponent=Get Body_Mesh())
   ) {layerSetup [{branchFilters [{boneName=clavicle_l}, {boneName=clavicle_r}, {boneName=upperarm_r}, {boneName=upperarm_l}]}]}
   Save cached pose 'StartPose'(Pose=@n6) {cacheName=StartPose}
  ]}
 EventGraph {nodes [
   Event OnCharacterCustomizationLoaded()
   > n1 = Sequence()
   n2 = Cast To PlayerPawn_Athena(execute=@n1.then_1, Object=TryGetPawnOwner())
   > Set Body_Mesh(Body_Mesh=GetSkeletalMeshForPartType(self=@n2))
   n3 = Cast To PlayerPawn_Athena(execute=@n1.then_0, Object=TryGetPawnOwner())
   > Set Body_Mesh(Body_Mesh=GetSkeletalMeshForPartType(self=@n3))
   n4 = CheckForBackbling()
   > Set IsWearingBackpackOrCape(IsWearingBackpackOrCape:bool=OR Boolean!(A=@n4, B=@n4.bIsBackblingACape, C:bool=@n4.bIsBackblingLong) -> (ReturnValue)) -> (Output_Get:bool)
   Event BlueprintUpdateAnimation()
   > n5 = Gravity Override State Function DEPRECATED(CharacterPartAnimInstance=Self-Reference(), GameTimeInSeconds=GetGameTimeInSeconds(), GravityOverrideParams=Get Gravity Override Params())
   > Set GravityOverride(GravityOverride=@n5)
  ]}
]
variables [
 Body_Mesh {type=SkeletalMeshComponent}
 Gravity Override Params {type=GravityOverrideParamsStruct}
 GravityOverride {type=Vector}
 IsWearingBackpackOrCape {type=boolean}
]
